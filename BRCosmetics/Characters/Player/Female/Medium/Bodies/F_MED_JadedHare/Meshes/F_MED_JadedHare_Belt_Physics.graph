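# Physics asset for the belt mesh of the JadedHare medium female player body, defining a dangling five-bone strap chain.
# Capsules on pelvis, spine_01 and both thighs give body volumes around the strap.
# The strap root, spine_01 and thigh_r lock angularly to the pelvis, while thigh_l has no constraint.
# Later strap links allow progressively wider swing toward the tip with very tight twist limits.
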PhysicsAsset /BRCosmetics/Characters/Player/Female/Medium/Bodies/F_MED_JadedHare/Meshes/F_MED_JadedHare_Belt_Physics
bodies {
 dyn_beltstrap_01 [{
   shapeType=Capsule
   length=1.0296
   radius=0.9329
   shapeName=dyn_beltstrap_01_capsule
   center=<0.9979 -0.0845 -0.2014>
   rotation=<rot 0 -96.1233 -85.1813>
  }]
 dyn_beltstrap_02 [{shapeType=Capsule, length=4, radius=1.5, shapeName=dyn_beltstrap_02_capsule, center=<2 0 0>, rotation=<rot 90 0 0>}]
 dyn_beltstrap_03 [{shapeType=Capsule, length=6, radius=1.5, shapeName=dyn_beltstrap_03_capsule, center=<2 0 0>, rotation=<rot 90 0 0>}]
 dyn_beltstrap_04 [{shapeType=Capsule, length=6, radius=1.5, shapeName=dyn_beltstrap_04_capsule, center=<2 0 0>, rotation=<rot 90 0 0>}]
 dyn_beltstrap_05 [{shapeType=Capsule, length=6, radius=1.5, shapeName=dyn_beltstrap_05_capsule, center=<2 0 0>, rotation=<rot 90 0 0>}]
 pelvis [{
   shapeType=Capsule
   length=16.46
   radius=8.6842
   shapeName=pelvis_capsule
   center=<-4.8901 -0.3907 -0.0077>
   rotation=<rot 0 -12.9877 -173.1801>
  }]
 spine_01 [{shapeType=Capsule, length=12.6111, radius=9.1241, shapeName=spine_01_capsule, center=<0.9567 -0.4177 0>, rotation=<rot 0 -90 0>}]
 thigh_l [{shapeType=Capsule, length=28.5374, radius=8.6457, center=<-13.5262 0.7174 1.501>, rotation=<rot -0.1341 -90.4229 85.8478>}]
 thigh_r [{shapeType=Capsule, length=28.5374, radius=8.6457, center=<13.5262 -0.7174 -1.501>, rotation=<rot -0.1341 -90.4229 -94.1522>}]
}
constraints [
 {
  bone1Name=thigh_r
  bone2Name=pelvis
  swing1LimitDegrees=45
  swing1Motion=Locked
  swing2LimitDegrees=45
  swing2Motion=Locked
  twistLimitDegrees=45
  twistMotion=Locked
 }
 {
  bone1Name=spine_01
  bone2Name=pelvis
  swing1LimitDegrees=45
  swing1Motion=Locked
  swing2LimitDegrees=45
  swing2Motion=Locked
  twistLimitDegrees=45
  twistMotion=Locked
 }
 {
  bone1Name=dyn_beltstrap_01
  bone2Name=pelvis
  swing1LimitDegrees=45
  swing1Motion=Locked
  swing2LimitDegrees=45
  swing2Motion=Locked
  twistLimitDegrees=45
  twistMotion=Locked
 }
 {bone1Name=dyn_beltstrap_02, bone2Name=dyn_beltstrap_01, swing1LimitDegrees=20, swing2LimitDegrees=20, twistLimitDegrees=2}
 {bone1Name=dyn_beltstrap_03, bone2Name=dyn_beltstrap_02, swing1LimitDegrees=35, swing2LimitDegrees=35, twistLimitDegrees=2}
 {bone1Name=dyn_beltstrap_04, bone2Name=dyn_beltstrap_03, swing1LimitDegrees=40, swing2LimitDegrees=40, twistLimitDegrees=2}
 {bone1Name=dyn_beltstrap_05, bone2Name=dyn_beltstrap_04, swing1LimitDegrees=50, swing2LimitDegrees=50, twistLimitDegrees=2}
]
