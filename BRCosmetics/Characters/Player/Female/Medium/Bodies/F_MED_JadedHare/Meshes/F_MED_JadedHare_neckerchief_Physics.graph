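# Physics asset for the neckerchief of the female medium JadedHare character body, giving its bow bones capsule bodies for physics-driven motion.
# A three-bone dyn_bow chain hangs from spine_05, with capsules growing larger and swing limits widening toward the tip, and zero-degree twist limits.
# Capsules on spine_04 and spine_05 form the torso volume, with angularly locked constraints linking spine_05 to spine_04 and spine_04 to pelvis.
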PhysicsAsset /BRCosmetics/Characters/Player/Female/Medium/Bodies/F_MED_JadedHare/Meshes/F_MED_JadedHare_neckerchief_Physics
bodies {
 dyn_bow_01 [{shapeType=Capsule, length=1, radius=1.5, shapeName=dyn_bow_01_capsule, center=<2 0 -0.8>, rotation=<rot 0 0 90>}]
 dyn_bow_02 [{shapeType=Capsule, length=3, radius=2, shapeName=dyn_bow_02_capsule, center=<2.6843 0 -1.1>, rotation=<rot 0 0 90>}]
 dyn_bow_03 [{shapeType=Capsule, length=18, radius=4, shapeName=dyn_bow_03_capsule, center=<4 0 -2.6>, rotation=<rot 0 0 90>}]
 spine_04 [
  {
   shapeType=Capsule
   length=12.6564
   radius=7.4355
   shapeName=spine_04_capsule
   center=<5.4085 2.4341 0.0082>
   rotation=<rot 0.5192 -18.2327 -179.1065>
  }
  {shapeType=Capsule, length=15, radius=9.8318, shapeName=spine_04_capsule, center=<0 2.4512 0>}
 ]
 spine_05 [{shapeType=Capsule, length=10.2487, radius=7.2619, shapeName=spine_05_capsule, center=<1.8909 0.4673 0.0808>}]
}
constraints [
 {
  bone1Name=spine_05
  bone2Name=spine_04
  swing1LimitDegrees=45
  swing1Motion=Locked
  swing2LimitDegrees=45
  swing2Motion=Locked
  twistLimitDegrees=45
  twistMotion=Locked
 }
 {
  bone1Name=spine_04
  bone2Name=pelvis
  swing1LimitDegrees=45
  swing1Motion=Locked
  swing2LimitDegrees=45
  swing2Motion=Locked
  twistLimitDegrees=45
  twistMotion=Locked
 }
 {bone1Name=dyn_bow_01, bone2Name=spine_05, swing1LimitDegrees=20, swing2LimitDegrees=20, twistLimitDegrees=0}
 {bone1Name=dyn_bow_02, bone2Name=dyn_bow_01, swing1LimitDegrees=30, swing2LimitDegrees=30, twistLimitDegrees=0}
 {bone1Name=dyn_bow_03, bone2Name=dyn_bow_02, swing1LimitDegrees=40, swing2LimitDegrees=40, twistLimitDegrees=0}
]
