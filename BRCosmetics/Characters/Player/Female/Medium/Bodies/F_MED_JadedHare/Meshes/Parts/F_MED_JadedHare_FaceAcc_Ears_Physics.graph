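# Physics asset for the face accessory ears of the JadedHare medium female player body, setting up simulation of ears and mask hair.
# Capsule bodies cover the head, four bones per ear and three mask hair bones per side.
# Each ear is a chain from the head with zero twist and swing that loosens toward the tip.
# Mask hair bones each constrain directly to the head rather than chaining, with slight twist allowed.
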PhysicsAsset /BRCosmetics/Characters/Player/Female/Medium/Bodies/F_MED_JadedHare/Meshes/Parts/F_MED_JadedHare_FaceAcc_Ears_Physics
bodies {
 dyn_ear_01_l [{shapeType=Capsule, length=5.1639, radius=1.4335, center=<3.2928 -0.0592 -0.0595>, rotation=<rot 0.8224 -77.4183 -81.7628>}]
 dyn_ear_01_r [{
   shapeType=Capsule
   length=5.0207
   radius=1.2188
   shapeName=dyn_ear_01_r_capsule
   center=<3.9429 -0.3696 -0.812>
   rotation=<rot 17.3643 81.2374 -88.6986>
  }]
 dyn_ear_02_l [{
   shapeType=Capsule
   length=5.5804
   radius=1.4153
   shapeName=dyn_ear_02_l_capsule
   center=<5.6977 -0.1139 0.0486>
   rotation=<rot 0.2479 -90.8319 -93.1123>
  }]
 dyn_ear_02_r [{
   shapeType=Capsule
   length=3.728
   radius=1.5926
   shapeName=dyn_ear_02_r_capsule
   center=<2.7694 -0.2089 0.0611>
   rotation=<rot 0 -85.8408 -86.967>
  }]
 dyn_ear_03_l [{
   shapeType=Capsule
   length=6.5151
   radius=1.599
   shapeName=dyn_ear_03_l_capsule
   center=<4.0919 0.4898 -0.4838>
   rotation=<rot 0.8643 86.7103 -90.6319>
  }]
 dyn_ear_03_r [{
   shapeType=Capsule
   length=4.6685
   radius=1.5209
   shapeName=dyn_ear_03_r_capsule
   center=<1.6653 0.72 -0.0395>
   rotation=<rot -18.2955 89.568 -90.0916>
  }]
 dyn_ear_04_l [{
   shapeType=Capsule
   length=6.3197
   radius=1.587
   shapeName=dyn_ear_04_l_capsule
   center=<2.8048 -1.0832 -0.1533>
   rotation=<rot -2.2716 68.2141 -85.1482>
  }]
 dyn_ear_04_r [{
   shapeType=Capsule
   length=5.123
   radius=1.4715
   shapeName=dyn_ear_04_r_capsule
   center=<3.1242 -1.2284 0.559>
   rotation=<rot 0 61.7647 -80.2038>
  }]
 dyn_maskhair_01_l [{
   shapeType=Capsule
   length=1.0935
   radius=0.7665
   shapeName=dyn_maskhair_01_l_capsule
   center=<0.6951 -0.075 -0.145>
   rotation=<rot 0 92.0324 -89.4862>
  }]
 dyn_maskhair_01_r [{
   shapeType=Capsule
   length=1.0935
   radius=0.7665
   shapeName=dyn_maskhair_01_r_capsule
   center=<1.5667 0.044 -0.1966>
   rotation=<rot 0 87.9675 -76.4535>
  }]
 dyn_maskhair_02_l [{
   shapeType=Capsule
   length=1.3018
   radius=0.7516
   shapeName=dyn_maskhair_02_l_capsule
   center=<1.5368 -0.3706 -0.0961>
   rotation=<rot 0 86.037 -94.1037>
  }]
 dyn_maskhair_02_r [{
   shapeType=Capsule
   length=1.3018
   radius=0.7516
   shapeName=dyn_maskhair_02_r_capsule
   center=<2.0009 0.4028 -0.4225>
   rotation=<rot 0 93.963 -82.2196>
  }]
 dyn_maskhair_03_l [{shapeType=Capsule, length=0.505, radius=0.505, shapeName=dyn_maskhair_03_l_capsule, center=<0 0.1212 0>, rotation=<rot 0 89.9998 -90>}]
 dyn_maskhair_03_r [{shapeType=Capsule, length=0.505, radius=0.505, shapeName=dyn_maskhair_03_r_capsule, center=<0 -0.121 0>, rotation=<rot 0 90 90>}]
 head [{shapeType=Capsule, length=2.3806, radius=7.8199, center=<7.052 1.6054 0.3502>, rotation=<rot -3.7463 -160.6668 -90.5971>}]
}
constraints [
 {bone1Name=dyn_ear_01_l, bone2Name=head, swing1LimitDegrees=4, swing2LimitDegrees=5, twistLimitDegrees=0}
 {bone1Name=dyn_ear_01_r, bone2Name=head, swing1LimitDegrees=4, swing2LimitDegrees=5, twistLimitDegrees=0}
 {bone1Name=dyn_maskhair_01_l, bone2Name=head, swing1LimitDegrees=10, swing2LimitDegrees=10, twistLimitDegrees=2}
 {bone1Name=dyn_maskhair_02_l, bone2Name=head, swing1LimitDegrees=10, swing2LimitDegrees=10, twistLimitDegrees=2}
 {bone1Name=dyn_maskhair_03_l, bone2Name=head, swing1LimitDegrees=10, swing2LimitDegrees=10, twistLimitDegrees=2}
 {bone1Name=dyn_maskhair_01_r, bone2Name=head, swing1LimitDegrees=10, swing2LimitDegrees=10, twistLimitDegrees=2}
 {bone1Name=dyn_maskhair_02_r, bone2Name=head, swing1LimitDegrees=10, swing2LimitDegrees=10, twistLimitDegrees=2}
 {bone1Name=dyn_maskhair_03_r, bone2Name=head, swing1LimitDegrees=10, swing2LimitDegrees=10, twistLimitDegrees=2}
 {bone1Name=dyn_ear_02_l, bone2Name=dyn_ear_01_l, swing1LimitDegrees=8, swing2LimitDegrees=5, twistLimitDegrees=0}
 {bone1Name=dyn_ear_03_l, bone2Name=dyn_ear_02_l, swing1LimitDegrees=10, swing2LimitDegrees=5, twistLimitDegrees=0}
 {bone1Name=dyn_ear_04_l, bone2Name=dyn_ear_03_l, swing1LimitDegrees=12, swing2LimitDegrees=5, twistLimitDegrees=0}
 {bone1Name=dyn_ear_02_r, bone2Name=dyn_ear_01_r, swing1LimitDegrees=8, swing2LimitDegrees=5, twistLimitDegrees=0}
 {bone1Name=dyn_ear_03_r, bone2Name=dyn_ear_02_r, swing1LimitDegrees=10, swing2LimitDegrees=5, twistLimitDegrees=0}
 {bone1Name=dyn_ear_04_r, bone2Name=dyn_ear_03_r, swing1LimitDegrees=12, swing2LimitDegrees=5, twistLimitDegrees=0}
]
