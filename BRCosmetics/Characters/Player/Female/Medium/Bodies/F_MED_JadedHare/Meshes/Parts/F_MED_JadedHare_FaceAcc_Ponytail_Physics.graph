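# Physics asset for the ponytail face-accessory mesh of the female medium JadedHare player body, defining capsule bodies for secondary hair motion.
# A ponytail chain hangs off the head, with swing limits widening toward the tip and minimal twist.
# A shorter two-bone left hair strand is more tightly limited.
# Upper spine, neck and head capsules are joined by fully locked constraints, with another lock between spine_05 and base_ponytail.
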PhysicsAsset /BRCosmetics/Characters/Player/Female/Medium/Bodies/F_MED_JadedHare/Meshes/Parts/F_MED_JadedHare_FaceAcc_Ponytail_Physics
bodies {
 base_ponytail [{shapeType=Capsule, length=2, radius=2, shapeName=base_ponytail_capsule, center=<1.062 -0.7672 0>, rotation=<rot 0 -71.0983 -90>}]
 dyn_hair_01_l [{
   shapeType=Capsule
   length=1.694
   radius=0.6588
   shapeName=dyn_hair_01_l_capsule
   center=<1.3421 0.1123 0.0243>
   rotation=<rot -1.2843 85.1424 -91.6652>
  }]
 dyn_hair_02_l [{
   shapeType=Capsule
   length=1.3765
   radius=0.5
   shapeName=dyn_hair_02_l_capsule
   center=<0.6227 0.1227 0.1315>
   rotation=<rot -0.051 90.1347 -82.1384>
  }]
 dyn_ponytail_01 [{
   shapeType=Capsule
   length=1.7234
   radius=1.9237
   shapeName=dyn_ponytail_01_capsule
   center=<1.5209 -0.4822 -0.1813>
   rotation=<rot 18.5926 -84.4152 -83.6871>
  }]
 dyn_ponytail_02 [{
   shapeType=Capsule
   length=1.0657
   radius=2.0535
   shapeName=dyn_ponytail_02_capsule
   center=<1.6026 0.1176 -0.2218>
   rotation=<rot 8.193 -82.8286 94.3012>
  }]
 dyn_ponytail_03 [{
   shapeType=Capsule
   length=0.8014
   radius=0.9979
   shapeName=dyn_ponytail_03_capsule
   center=<1.2852 0.0002 -0.0769>
   rotation=<rot -0.2425 -85.9312 93.414>
  }]
 dyn_ponytail_04 [{
   shapeType=Capsule
   length=1.0159
   radius=0.803
   shapeName=dyn_ponytail_04_capsule
   center=<0.8704 -0.0823 -0.449>
   rotation=<rot 0.8356 93.7328 -90.2872>
  }]
 head [{shapeType=Capsule, length=3.5902, radius=5.6507, center=<-0.7426 1.8591 0.2285>, rotation=<rot 0.8184 -95.0633 -89.6368>}]
 neck_01 [{shapeType=Capsule, length=16.0683, radius=5.5406, shapeName=neck_01_capsule, center=<6.3891 0.6321 0>, rotation=<rot 0 -90 0>}]
 neck_02 [{
   shapeType=Capsule
   length=11.4665
   radius=5.4169
   shapeName=neck_02_capsule
   center=<-2.5643 0.0212 -0.0094>
   rotation=<rot 0 -90.0002 -0.0752>
  }]
 spine_05 [{shapeType=Capsule, length=13.927, radius=8.2133, shapeName=spine_05_capsule, center=<4.7541 -0.7059 0>, rotation=<rot 0 -90 0>}]
}
constraints [
 {
  bone1Name=head
  bone2Name=neck_02
  swing1LimitDegrees=45
  swing1Motion=Locked
  swing2LimitDegrees=45
  swing2Motion=Locked
  twistLimitDegrees=45
  twistMotion=Locked
 }
 {
  bone1Name=neck_01
  bone2Name=spine_05
  swing1LimitDegrees=45
  swing1Motion=Locked
  swing2LimitDegrees=45
  swing2Motion=Locked
  twistLimitDegrees=45
  twistMotion=Locked
 }
 {
  bone1Name=neck_02
  bone2Name=neck_01
  swing1LimitDegrees=45
  swing1Motion=Locked
  swing2LimitDegrees=45
  swing2Motion=Locked
  twistLimitDegrees=45
  twistMotion=Locked
 }
 {bone1Name=base_ponytail, bone2Name=head, swing1LimitDegrees=30, swing2LimitDegrees=20, twistLimitDegrees=0}
 {bone1Name=dyn_ponytail_01, bone2Name=base_ponytail, swing1LimitDegrees=20, swing2LimitDegrees=20, twistLimitDegrees=2}
 {bone1Name=dyn_ponytail_02, bone2Name=dyn_ponytail_01, swing1LimitDegrees=30, swing2LimitDegrees=30, twistLimitDegrees=2}
 {bone1Name=dyn_ponytail_03, bone2Name=dyn_ponytail_02, swing1LimitDegrees=40, swing2LimitDegrees=40, twistLimitDegrees=2}
 {bone1Name=dyn_ponytail_04, bone2Name=dyn_ponytail_03, swing1LimitDegrees=50, swing2LimitDegrees=50, twistLimitDegrees=2}
 {
  bone1Name=spine_05
  bone2Name=base_ponytail
  swing1LimitDegrees=45
  swing1Motion=Locked
  swing2LimitDegrees=45
  swing2Motion=Locked
  twistLimitDegrees=45
  twistMotion=Locked
 }
 {bone1Name=dyn_hair_01_l, bone2Name=head, swing1LimitDegrees=10, swing2LimitDegrees=10, twistLimitDegrees=2}
 {bone1Name=dyn_hair_02_l, bone2Name=dyn_hair_01_l, swing1LimitDegrees=10, swing2LimitDegrees=10, twistLimitDegrees=2}
]
