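# Animation Blueprint for the Jaded Hare female cosmetic's face accessory, adding secondary physics motion to its ears and ponytail.
# It copies the pose from the attached parent mesh, then runs two chained RigidBody simulations in base-bone space with separate physics assets.
# The ears simulate relative to neck_01 with reduced gravity and tight motion clamps; the ponytail simulates relative to spine_05 with full gravity and looser limits.
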AnimBlueprint /BRCosmetics/Characters/Player/Female/Medium/Bodies/F_MED_JadedHare/Meshes/Parts/F_MED_JadedHare_FaceAcc_AnimBP
target_skeleton=&/F_MED_JadedHare_FaceAcc_Skeleton
graphs [
 AnimGraph {nodes [
   n1 = RigidBody(ComponentPose=Local To Component(LocalPose=Copy Pose From Mesh() {bUseAttachedParent=true})) {
    bOverrideWorldGravity=true
    overridePhysicsAsset=&/F_MED_JadedHare_FaceAcc_Ears_Physics
    simulationSpace=BaseBoneSpace
    alphaBoolBlend {blendOption=Linear}
    baseBoneRef {boneName=neck_01}
    overrideWorldGravity {z=-700}
    simSpaceSettings {maxAngularAcceleration=40, maxAngularVelocity=8, maxLinearAcceleration=100, maxLinearVelocity=200, velocityScaleZ=0.25, worldAlpha=0.7}
   }
   n2 = RigidBody(ComponentPose=@n1) {
    bOverrideWorldGravity=true
    overridePhysicsAsset=&/F_MED_JadedHare_FaceAcc_Ponytail_Physics
    simulationSpace=BaseBoneSpace
    alphaBoolBlend {blendOption=Linear}
    baseBoneRef {boneName=spine_05}
    overrideWorldGravity {z=-980}
    simSpaceSettings {maxAngularAcceleration=400, maxAngularVelocity=80, maxLinearAcceleration=5000, maxLinearVelocity=1000, velocityScaleZ=0.7, worldAlpha=0.6}
   }
   Output Pose(Result=Component To Local(ComponentPose=@n2)) {layerGroup=None}
  ]}
 EventGraph {nodes [Event BlueprintUpdateAnimation()]}
]
variables []
